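# Control Rig for the Sweet Teriyaki Orange backpack that squashes and stretches the pack body based on the jiggle bone's motion.
# It reads the absolute local Z translation of dyn_jiggle and remaps it into weights for two chained Modify Transforms on dyn_body.
# Higher offsets blend in a slight Z stretch with X narrowing, while lower offsets blend in a stronger Z squash with X widening.
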ControlRigBlueprint /BRCosmetics/Accessories/FORT_Backpacks/Backpack_M_MED_Sweet_Teriyaki_Orange/Meshes/M_MED_Sweet_Teriyaki_Orange_Pack_CtrlRig
graphs [
 RigVMFunctionLibrary {links [], nodes []}
 RigVMModel {nodes [
   RigUnit_BeginExecution = &RigUnit_BeginExecution()
   n1 = /Script/RigVM.RigVMFunction_MathFloatAbs(Value=&RigUnit_GetTransform(Item=(Type=Bone,Name="dyn_jiggle"), Space=LocalSpace, bInitial=false).Transform.Translation.Z) {is=struct RigVMUnitNode, template="Absolute::Execute(in Value,out Result)"}
   &RigUnit_ModifyTransforms(
    ItemToModify=((Item=(Type=Bone),Transform=(Rotation=(X=0,Y=0,Z=0,W=1),Translation=(X=0,Y=0,Z=0),Scale3D=(X=1,Y=1,Z=1))),(Item=(Type=Bone,Name="dyn_body"),Transform=(Rotation=(X=0,Y=0,Z=0,W=1),Translation=(X=0,Y=0,Z=0),Scale3D=(X=0.975,Y=1,Z=1.025))))
    WeightMinimum=0
    Weight=&RigVMFunction_MathFloatRemap(SourceMaximum=8, SourceMinimum=4, Value=@n1.Result).Result
    ExecutePin=@RigUnit_BeginExecution.ExecutePin
   )
   > &RigUnit_ModifyTransforms(
    ItemToModify=((Item=(Type=Bone),Transform=(Rotation=(X=0,Y=0,Z=0,W=1),Translation=(X=0,Y=0,Z=0),Scale3D=(X=1,Y=1,Z=1))),(Item=(Type=Bone,Name="dyn_body"),Transform=(Rotation=(X=0,Y=0,Z=0,W=1),Translation=(X=0,Y=0,Z=0),Scale3D=(X=1.1,Y=1,Z=0.9))))
    WeightMinimum=0
    Weight=&RigVMFunction_MathFloatRemap(SourceMaximum=2, SourceMinimum=4, Value=@n1.Result).Result
   )
  ]}
]
hierarchy [
 root {loc=<0 0 0>, rot=<0 0 0 1>}
 pack {parent=root, loc=<0 -0.7316 0>, rot=<0 0 0 1>}
 dyn_body {parent=pack, loc=<-4.5828 -11.6499 -4.2263>, rot=<0 0 0 1>}
 dyn_fin_r {parent=dyn_body, loc=<-3.4566 -9.8609 -8.3344>, rot=<0.6716 -0.5962 -0.292 -0.329>}
 dyn_fin_l {parent=dyn_body, loc=<-3.4566 9.8609 -8.3345>, rot=<0.329 -0.292 -0.5962 -0.6716>}
 dyn_tail {parent=dyn_body, loc=<16.4854 0 -1.7807>, rot=<0.707 -0.0114 -0.0114 -0.7071>}
 dyn_dorsal {parent=dyn_body, loc=<1.9152 0.0001 12.8136>, rot=<-0.5164 0.483 -0.483 -0.5164>}
 dyn_jiggle {parent=dyn_body, loc=<0 0 0>, rot=<0 0 0 1>}
]
