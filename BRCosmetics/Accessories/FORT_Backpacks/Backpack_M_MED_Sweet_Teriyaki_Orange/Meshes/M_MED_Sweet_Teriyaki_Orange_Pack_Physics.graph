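# Physics asset for the Sweet Teriyaki Orange backpack cosmetic, giving simulation bodies to a creature-like pack with fins, tail, dorsal and eyes.
# Capsules cover each part, with a small sphere on the jiggle bone that is rotationally locked to the central body.
# Fins, tail and dorsal form two-link chains off the body with tight twist and uneven swing limits, wider on right fins than left.
# Both eyes attach directly to the body with wider swing limits.
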PhysicsAsset /BRCosmetics/Accessories/FORT_Backpacks/Backpack_M_MED_Sweet_Teriyaki_Orange/Meshes/M_MED_Sweet_Teriyaki_Orange_Pack_Physics
bodies {
 dyn_body [{shapeType=Capsule, length=5, radius=3, shapeName=dyn_body_capsule, center=<2 0 0>, rotation=<rot 90 0 0>}]
 dyn_dorsal1 [{shapeType=Capsule, length=8.9338, radius=2.1272, shapeName=dyn_dorsal1_capsule, center=<1 0 0>, rotation=<rot 0 0 90>}]
 dyn_dorsal2 [{shapeType=Capsule, length=6.5811, radius=2.4576, shapeName=dyn_dorsal2_capsule, center=<1 0 0>, rotation=<rot 0 0 90>}]
 dyn_eye_l [{shapeType=Capsule, length=3, radius=4, shapeName=dyn_eye_l_capsule, center=<1 0 0>, rotation=<rot 90 0 0>}]
 dyn_eye_r [{shapeType=Capsule, length=3, radius=4, shapeName=dyn_eye_r_capsule, center=<1 0 0>, rotation=<rot 90 0 0>}]
 dyn_fin_l1 [{shapeType=Capsule, length=2.2378, radius=1.5127, shapeName=dyn_fin_l1_capsule, center=<0.5 0 0>, rotation=<rot 90 0 0>}]
 dyn_fin_l2 [{shapeType=Capsule, length=2.8658, radius=1.6783, shapeName=dyn_fin_l2_capsule, center=<0.5 0 0>, rotation=<rot 90 0 0>}]
 dyn_fin_r1 [{shapeType=Capsule, length=2.2379, radius=1.5128, shapeName=dyn_fin_r1_capsule, center=<0.5 0 0>, rotation=<rot 90 0 0>}]
 dyn_fin_r2 [{shapeType=Capsule, length=2.8659, radius=1.6782, shapeName=dyn_fin_r2_capsule, center=<0.5 0 0>, rotation=<rot 90 0 0>}]
 dyn_jiggle [{shapeType=Sphere, radius=3, shapeName=dyn_jiggle_sphere}]
 dyn_tail1 [{shapeType=Capsule, length=6.919, radius=3.0505, shapeName=dyn_tail1_capsule, center=<1 0 0>, rotation=<rot 0 0 90>}]
 dyn_tail2 [{shapeType=Capsule, length=9.3553, radius=2.9081, shapeName=dyn_tail2_capsule, center=<1 0 0>, rotation=<rot 0 0 90>}]
}
constraints [
 {
  bone1Name=dyn_jiggle
  bone2Name=dyn_body
  swing1LimitDegrees=45
  swing1Motion=Locked
  swing2LimitDegrees=45
  swing2Motion=Locked
  twistLimitDegrees=45
  twistMotion=Locked
 }
 {bone1Name=dyn_fin_r1, bone2Name=dyn_body, swing1LimitDegrees=6, swing2LimitDegrees=25, twistLimitDegrees=1}
 {bone1Name=dyn_fin_r2, bone2Name=dyn_fin_r1, swing1LimitDegrees=6, swing2LimitDegrees=25, twistLimitDegrees=1}
 {bone1Name=dyn_fin_l1, bone2Name=dyn_body, swing1LimitDegrees=6, swing2LimitDegrees=15, twistLimitDegrees=1}
 {bone1Name=dyn_fin_l2, bone2Name=dyn_fin_l1, swing1LimitDegrees=6, swing2LimitDegrees=15, twistLimitDegrees=1}
 {bone1Name=dyn_tail1, bone2Name=dyn_body, swing1LimitDegrees=5, swing2LimitDegrees=25, twistLimitDegrees=1}
 {bone1Name=dyn_tail2, bone2Name=dyn_tail1, swing1LimitDegrees=5, swing2LimitDegrees=25, twistLimitDegrees=1}
 {bone1Name=dyn_dorsal1, bone2Name=dyn_body, swing1LimitDegrees=5, swing2LimitDegrees=25, twistLimitDegrees=1}
 {bone1Name=dyn_dorsal2, bone2Name=dyn_dorsal1, swing1LimitDegrees=5, swing2LimitDegrees=25, twistLimitDegrees=1}
 {bone1Name=dyn_eye_r, bone2Name=dyn_body, swing1LimitDegrees=45, swing2LimitDegrees=45, twistLimitDegrees=1}
 {bone1Name=dyn_eye_l, bone2Name=dyn_body, swing1LimitDegrees=45, swing2LimitDegrees=45, twistLimitDegrees=1}
]
